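# Sound cue for the idle voice-over of the OuterTowerBoss NPC from the Outter NPC library.
# A single Random node picks among six idle wave variants with equal weights and is excluded from branch culling.
# Playback uses the shared close-range NPC voice attenuation and the critical enemy VO sound class.
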SoundCue /NPCLibrary/NPCs/Outter/Outter/Audio/VO/OuterTowerBoss_2021_Idle_Cue
root=Random_0
nodes [
 {
  id=Random_0
  children [WavePlayer_0,WavePlayer_1,WavePlayer_2,WavePlayer_3,WavePlayer_4,WavePlayer_5]
  properties {bShouldExcludeFromBranchCulling=true, weights [1,1,1,1,1,1]}
 }
 {id=WavePlayer_0, properties {soundWaveAssetPtr=/NPCLibrary/NPCs/Outter/Outter/Audio/VO/Waves/OuterTowerBoss_2021_Idle_010}}
 {id=WavePlayer_1, properties {soundWaveAssetPtr=/NPCLibrary/NPCs/Outter/Outter/Audio/VO/Waves/OuterTowerBoss_2021_Idle_020}}
 {id=WavePlayer_2, properties {soundWaveAssetPtr=/NPCLibrary/NPCs/Outter/Outter/Audio/VO/Waves/OuterTowerBoss_2021_Idle_030}}
 {id=WavePlayer_3, properties {soundWaveAssetPtr=/NPCLibrary/NPCs/Outter/Outter/Audio/VO/Waves/OuterTowerBoss_2021_Idle_040}}
 {id=WavePlayer_4, properties {soundWaveAssetPtr=/NPCLibrary/NPCs/Outter/Outter/Audio/VO/Waves/OuterTowerBoss_2021_Idle_050}}
 {id=WavePlayer_5, properties {soundWaveAssetPtr=/NPCLibrary/NPCs/Outter/Outter/Audio/VO/Waves/OuterTowerBoss_2021_Idle_060}}
]
properties {
 attenuationSettings=/Game/Sounds/Fort_Human_VO_BR/NPCs/Shared/NPC_VO_Close_Att
 maxDistance=8500
 soundClassObject=&Fort_VO_Enemy_Critical
 volumeMultiplier=1
}
